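# Physics asset for the necklace on the medium male Crowcall character body, covering only the necklace bones and their spine_05 anchor.
# Small capsules form a chain from spine_05 through necklace_01 to necklace_02, with limited swing and locked twist.
# Two much larger, offset capsules on the necklace pivot bones each constrain directly to spine_05 with free swing and twist, separate from the chain.
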PhysicsAsset /BRCosmetics/Characters/Player/Male/Medium/Bodies/M_MED_Crowcall/Meshes/M_MED_CrowCall_Physics
bodies {
 necklace_01 [{shapeType=Capsule, length=2, radius=1, shapeName=necklace_01_capsule, center=<0 0 0.55>, rotation=<rot 90 180 180>}]
 necklace_01_pivot [{
   shapeType=Capsule
   length=3.8528
   radius=17.1472
   shapeName=necklace_01_pivot_capsule
   center=<1.2218 0 -16.8879>
   rotation=<rot 88.4161 180 90>
  }]
 necklace_02 [{shapeType=Capsule, length=1.9293, radius=1.0707, shapeName=necklace_02_capsule, center=<1.1875 0 0.5344>, rotation=<rot 86.1949 0 0>}]
 necklace_02_pivot [{
   shapeType=Capsule
   length=15.5
   radius=7.5
   shapeName=necklace_02_pivot_capsule
   center=<-1.5908 0 -8.0692>
   rotation=<rot 90 -165.9638 -75.9638>
  }]
 spine_05 [{shapeType=Capsule, length=0.5, radius=0.5, shapeName=spine_05_capsule}]
}
constraints [
 {bone1Name=necklace_02, bone2Name=necklace_01, swing1LimitDegrees=15, swing2LimitDegrees=20, twistLimitDegrees=45, twistMotion=Locked}
 {bone1Name=necklace_01, bone2Name=spine_05, swing1LimitDegrees=15, swing2LimitDegrees=20, twistLimitDegrees=45, twistMotion=Locked}
 {
  bone1Name=necklace_02_pivot
  bone2Name=spine_05
  swing1LimitDegrees=45
  swing1Motion=Free
  swing2LimitDegrees=45
  swing2Motion=Free
  twistLimitDegrees=45
  twistMotion=Free
 }
 {
  bone1Name=necklace_01_pivot
  bone2Name=spine_05
  swing1LimitDegrees=45
  swing1Motion=Free
  swing2LimitDegrees=45
  swing2Motion=Free
  twistLimitDegrees=45
  twistMotion=Free
 }
]
