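# Post-processing AnimBlueprint for the Crowcall character body, adding secondary motion and pose fixes to the parent mesh's copied pose.
# Control Rig, weighted necklace constraints, AnimDynamics on shirt and bracelets, and RigidBody physics run in series.
# Interpolated additive bone offsets then correct back, necklace, bracelet and right foot.
# The update event derives their alphas from movement, weapon and hand yaw, plus gravity override and a sine-driven skydive necklace wave.
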AnimBlueprint /BRCosmetics/Characters/Player/Male/Medium/Bodies/M_MED_Crowcall/Meshes/M_MED_CrowCall_AnimBP
target_skeleton=/Game/Characters/Player/Male/Medium/Bodies/M_MED_Crowcall/Meshes/M_MED_CrowCall_Skeleton
graphs [
 AnimGraph {nodes [
   n1 = Control Rig!(
    Source=Copy Pose From Mesh() {bUseAttachedParent=true}
    Alpha
    SMG_And_Pistols:double=Get Alpha_IsSMG_and_Pistols_Float() -> (Alpha_IsSMG_and_Pistols_Float:double)
   ) -> (Pose) {
    alphaBoolBlend {blendOption=Linear}
    controlRigAssetReference {blueprintRigClass=&/M_MED_CrowCall_ControlRig.M_MED_CrowCall_ControlRig_C}
   }
   n2 = "Constraint - necklace_01_pivot (PPPPPPP)"!(
    ComponentPose=Local To Component(LocalPose=@n1)
    Alpha=1
    ConstraintWeights_0=0.4503
    ConstraintWeights_1:float=0.2942
    ConstraintWeights_2:float=0.1144
    ConstraintWeights_3:float=0.0467
    ConstraintWeights_4:float=0.0445
    ConstraintWeights_5:float=0.025
    ConstraintWeights_6:float=0.025
   ) -> (Pose) {
    alphaBoolBlend {blendOption=Linear}
    boneToModify {boneName=necklace_01_pivot}
    constraintSetup [
     {targetBone {boneName=spine_05}}
     {targetBone {boneName=spine_04}}
     {targetBone {boneName=spine_03}}
     {targetBone {boneName=deform_pec_r}}
     {targetBone {boneName=spine_02}}
     {targetBone {boneName=clavicle_l}}
     {targetBone {boneName=clavicle_r}}
    ]
    constraintWeights [1,1,1,1,1,1,1]
   }
   n3 = "Constraint - necklace_02_pivot (PPPPPP)"!(
    ComponentPose=@n2
    Alpha=1
    ConstraintWeights_0=0.9959
    ConstraintWeights_1:float=0.0026
    ConstraintWeights_2:float=0.0007
    ConstraintWeights_3:float=0.0003
    ConstraintWeights_4:float=0.0003
    ConstraintWeights_5:float=0.0001
   ) -> (Pose) {
    alphaBoolBlend {blendOption=Linear}
    boneToModify {boneName=necklace_02_pivot}
    constraintSetup [
     {targetBone {boneName=spine_05}}
     {targetBone {boneName=spine_04}}
     {targetBone {boneName=spine_03}}
     {targetBone {boneName=clavicle_l}}
     {targetBone {boneName=clavicle_r}}
     {targetBone {boneName=spine_02}}
    ]
    constraintWeights [1,1,1,1,1,1]
   }
   n4 = Get Gravity Override()
   n5 = Get Alpha_isFrontEnd() -> (Alpha_isFrontEnd:float)
   n6 = Anim Dynamics - Bone: dyn_shirt_l!(ComponentPose=@n3, Alpha=@n5, GravityOverride:Vector=@n4, ExternalForce="0,0,0") -> (Pose) {
    angularDampingOverride=1
    angularSpringConstant=20000
    bAngularSpring=true
    bOverrideAngularDamping=true
    bOverrideLinearDamping=true
    bUseGravityOverride=true
    gravityScale=0.35
    linearDampingOverride=0.9999
    simulationSpace=World
    alphaBoolBlend {blendOption=Linear}
    boundBone {boneName=dyn_shirt_l}
    componentLinearAccScale=<1 1 1>
    componentLinearVelScale=<1 1 1>
    gravityOverride {z=-980}
    physicsBodyDefinitions [{
      boundBone {boneName=dyn_shirt_l}
      boxExtents=<20 30 30>
      constraintSetup {angularLimitsMax=<0 10 10>, angularLimitsMin=<0 -10 -10>, angularTarget=<1 0 0>}
      localJointOffset=<2.5 0 0>
     }]
   }
   n7 = Anim Dynamics - Bone: dyn_shirt_b!(ComponentPose=@n6, Alpha=@n5, GravityOverride:Vector=@n4, ExternalForce="0,0,0") -> (Pose) {
    angularDampingOverride=1
    angularSpringConstant=20000
    bAngularSpring=true
    bOverrideAngularDamping=true
    bOverrideLinearDamping=true
    bUseGravityOverride=true
    gravityScale=0.35
    linearDampingOverride=0.9999
    simulationSpace=World
    alphaBoolBlend {blendOption=Linear}
    boundBone {boneName=dyn_shirt_b}
    componentLinearAccScale=<1 1 1>
    componentLinearVelScale=<1 1 1>
    gravityOverride {z=-980}
    physicsBodyDefinitions [{
      boundBone {boneName=dyn_shirt_b}
      boxExtents=<20 30 30>
      constraintSetup {angularLimitsMax=<0 10 10>, angularLimitsMin=<0 -10 -10>, angularTarget=<1 0 0>}
      localJointOffset=<2.5 0 0>
     }]
   }
   n8 = Anim Dynamics - Bone: dyn_shirt_r!(ComponentPose=@n7, Alpha=@n5, GravityOverride:Vector=@n4, ExternalForce="0,0,0") -> (Pose) {
    angularDampingOverride=1
    angularSpringConstant=20000
    bAngularSpring=true
    bOverrideAngularDamping=true
    bOverrideLinearDamping=true
    bUseGravityOverride=true
    gravityScale=0.35
    linearDampingOverride=0.9999
    simulationSpace=World
    alphaBoolBlend {blendOption=Linear}
    boundBone {boneName=dyn_shirt_r}
    componentLinearAccScale=<1 1 1>
    componentLinearVelScale=<1 1 1>
    gravityOverride {z=-980}
    physicsBodyDefinitions [{
      boundBone {boneName=dyn_shirt_r}
      boxExtents=<20 30 30>
      constraintSetup {angularLimitsMax=<0 10 10>, angularLimitsMin=<0 -10 -10>, angularTarget=<1 0 0>}
      localJointOffset=<-2.5 0 0>
     }]
   }
   n9 = Anim Dynamics - Bone: dyn_bracelet_01!(ComponentPose=@n8, Alpha=@n5, GravityOverride:Vector=@n4, ExternalForce="0,0,0") -> (Pose) {
    angularDampingOverride=1
    angularSpringConstant=20000
    bAngularSpring=true
    bOverrideAngularDamping=true
    bUseGravityOverride=true
    gravityScale=0.35
    simulationSpace=World
    alphaBoolBlend {blendOption=Linear}
    boundBone {boneName=dyn_bracelet_01}
    componentLinearAccScale=<1 1 1>
    componentLinearVelScale=<1 1 1>
    gravityOverride {z=-980}
    physicsBodyDefinitions [{
      boundBone {boneName=dyn_bracelet_01}
      boxExtents=<20 30 30>
      constraintSetup {angularLimitsMax=<5 5 5>, angularLimitsMin=<-5 -5 -5>, angularTarget=<1 0 0>}
      localJointOffset=<3 0 0>
     }]
   }
   n10 = Anim Dynamics - Bone: dyn_bracelet_02!(ComponentPose=@n9, Alpha=@n5, GravityOverride:Vector=@n4, ExternalForce="0,0,0") -> (Pose) {
    angularDampingOverride=1
    angularSpringConstant=20000
    bAngularSpring=true
    bOverrideAngularDamping=true
    bUseGravityOverride=true
    gravityScale=0.35
    simulationSpace=World
    alphaBoolBlend {blendOption=Linear}
    boundBone {boneName=dyn_bracelet_02}
    componentLinearAccScale=<1 1 1>
    componentLinearVelScale=<1 1 1>
    gravityOverride {z=-980}
    physicsBodyDefinitions [{
      boundBone {boneName=dyn_bracelet_02}
      boxExtents=<20 30 30>
      constraintSetup {angularLimitsMax=<5 5 5>, angularLimitsMin=<-5 -5 -5>, angularTarget=<1 0 0>}
      localJointOffset=<3 0 0>
     }]
   }
   n11 = RigidBody!(ComponentPose=@n10, Alpha, OverrideWorldGravity:Vector=Get Gravity Override(), ExternalForce) -> (Pose) {
    bOverrideWorldGravity=true
    overridePhysicsAsset=&/M_MED_CrowCall_Physics
    simulationSpace=BaseBoneSpace
    alphaBoolBlend {blendOption=Linear}
    alphaScaleBiasClamp {bInterpResult=true}
    baseBoneRef {boneName=spine_03}
    componentAppliedLinearAccClamp=<3000 3000 3000>
    componentLinearAccScale=<0.3 0.3 0.3>
    componentLinearVelScale=<0.3 0.3 0.3>
    simSpaceSettings {
     maxAngularAcceleration=1000
     maxAngularVelocity=3
     maxLinearAcceleration=700
     maxLinearVelocity=500
     velocityScaleZ=0
     worldAlpha=0.7
     externalLinearDragV=<0.2 0.2>
    }
   }
   n12 = "Transform (Modify) Bone - Bone: correct_back"(ComponentPose=@n11, Alpha=Get Alpha_isFloating() -> (Alpha_isFloating:float), Translation="0,-3,0", Rotation="0,0,0", Scale="1,1,1") {
    rotationMode=BMM_Additive
    rotationSpace=BCS_BoneSpace
    scaleMode=BMM_Additive
    scaleSpace=BCS_BoneSpace
    translationMode=BMM_Additive
    translationSpace=BCS_BoneSpace
    alphaBoolBlend {blendOption=Linear}
    alphaScaleBiasClamp {bInterpResult=true, interpSpeedDecreasing=5, interpSpeedIncreasing=5}
    boneToModify {boneName=correct_back}
    translation {y=-3}
   }
   n13 = Get Alpha_isMelee() -> (Alpha_isMelee:double)
   n14 = "Transform (Modify) Bone - Bone: correct_back"(ComponentPose=@n12, Alpha=@n13, Translation="0,-2,0", Rotation="0,0,0", Scale="1,1,1") {
    rotationMode=BMM_Additive
    rotationSpace=BCS_BoneSpace
    scaleMode=BMM_Additive
    scaleSpace=BCS_BoneSpace
    translationMode=BMM_Additive
    translationSpace=BCS_BoneSpace
    alphaBoolBlend {blendOption=Linear}
    alphaScaleBiasClamp {bInterpResult=true, interpSpeedDecreasing=5, interpSpeedIncreasing=5}
    boneToModify {boneName=correct_back}
    translation {y=-2}
   }
   n15 = "Transform (Modify) Bone - Bone: necklace_base"(ComponentPose=@n14, Alpha=@n13, Translation="-0.000,-0.000,0.773", Rotation="0,0,0", Scale="1,1,1") {
    rotationMode=BMM_Additive
    rotationSpace=BCS_BoneSpace
    scaleMode=BMM_Additive
    scaleSpace=BCS_BoneSpace
    translationMode=BMM_Additive
    translationSpace=BCS_BoneSpace
    alphaBoolBlend {blendOption=Linear}
    alphaScaleBiasClamp {bInterpResult=true, interpSpeedDecreasing=5, interpSpeedIncreasing=5}
    boneToModify {boneName=necklace_base}
    translation=<0 0 0.7734>
   }
   n16 = "Transform (Modify) Bone - Bone: necklace_base"(ComponentPose=@n15, Alpha=1, Translation="0,0,0", Rotation=Get rotatorWaveRope_Y() -> (rotatorWaveRope_Y:Rotator), Scale="1,1,1") {
    rotationMode=BMM_Additive
    rotationSpace=BCS_BoneSpace
    scaleMode=BMM_Additive
    scaleSpace=BCS_BoneSpace
    translationMode=BMM_Additive
    translationSpace=BCS_BoneSpace
    alphaBoolBlend {blendOption=Linear}
    alphaScaleBiasClamp {bInterpResult=true, interpSpeedDecreasing=5, interpSpeedIncreasing=5}
    boneToModify {boneName=necklace_base}
   }
   n17 = "Transform (Modify) Bone - Bone: bracelet"(ComponentPose=@n16, Alpha=Get Alpha_isHandRotate() -> (Alpha_isHandRotate:float), Translation="0,2,0", Rotation="0,-30,0", Scale="1,1,1") {
    rotationMode=BMM_Additive
    rotationSpace=BCS_BoneSpace
    scaleMode=BMM_Additive
    scaleSpace=BCS_BoneSpace
    translationMode=BMM_Additive
    translationSpace=BCS_BoneSpace
    alphaBoolBlend {blendOption=Linear}
    alphaScaleBiasClamp {bInterpResult=true, interpSpeedDecreasing=5, interpSpeedIncreasing=5}
    boneToModify {boneName=bracelet}
    rotation {yaw=-30}
    translation {y=2}
   }
   n18 = "Transform (Modify) Bone - Bone: correct_back"(
    ComponentPose=@n17
    Alpha=Get Alpha_isOpenParachute() -> (Alpha_isOpenParachute:float)
    Translation="0,-2,0"
    Rotation="0,0,0"
    Scale="1,1,1"
   ) {
    rotationMode=BMM_Additive
    rotationSpace=BCS_BoneSpace
    scaleMode=BMM_Additive
    scaleSpace=BCS_BoneSpace
    translationMode=BMM_Additive
    translationSpace=BCS_BoneSpace
    alphaBoolBlend {blendOption=Linear}
    alphaScaleBiasClamp {bInterpResult=true, interpSpeedDecreasing=5}
    boneToModify {boneName=correct_back}
    translation {y=-2}
   }
   n19 = "Transform (Modify) Bone - Bone: foot_correct_r"(
    ComponentPose=@n18
    Alpha=Get Alpha_isCrouching() -> (Alpha_isCrouching:float)
    Translation="3.345,0.666,-3.947"
    Rotation="-20.000,-0.000,0.000"
    Scale="1,1,1"
   ) {
    rotationMode=BMM_Additive
    rotationSpace=BCS_BoneSpace
    scaleMode=BMM_Additive
    scaleSpace=BCS_BoneSpace
    translationMode=BMM_Additive
    translationSpace=BCS_BoneSpace
    alphaBoolBlend {blendOption=Linear}
    alphaScaleBiasClamp {bInterpResult=true, interpSpeedDecreasing=5}
    boneToModify {boneName=foot_correct_r}
    rotation=<rot -19.9999 0 0.0001>
    translation=<3.345 0.6655 -3.947>
   }
   n20 = "Transform (Modify) Bone - Bone: correct_back"(
    ComponentPose=@n19
    Alpha=Get Alpha_IS_DBNO / Emote() -> (Alpha_IS_DBNO / Emote:double)
    Translation="0,-2,0"
    Rotation="0,0,0"
    Scale="1,1,1"
   ) {
    rotationMode=BMM_Additive
    rotationSpace=BCS_BoneSpace
    scaleMode=BMM_Additive
    scaleSpace=BCS_BoneSpace
    translationMode=BMM_Additive
    translationSpace=BCS_BoneSpace
    alphaBoolBlend {blendOption=Linear}
    alphaScaleBiasClamp {bInterpResult=true, interpSpeedDecreasing=5}
    boneToModify {boneName=correct_back}
    translation {y=-2}
   }
   Output Pose(Result=Component To Local(ComponentPose=@n20)) {layerGroup=None}
  ]}
 EventGraph {nodes [
   Event BlueprintUpdateAnimation()
   > n1 = Sequence!(execute) -> (
    then_0
    then_1
    then_2:Exec
    then_3:Exec
    then_4:Exec
    then_5:Exec
    then_6:Exec
    then_7:Exec
    then_8:Exec
    then_9:Exec
    then_10:Exec
    then_11:Exec
    then_12:Exec
   )
   n2 = Gravity Override State Function DEPRECATED(
    execute=@n1.then_2
    CharacterPartAnimInstance=Self-Reference()
    GameTimeInSeconds=GetGameTimeInSeconds()
    GravityOverrideParams=Get Gravity Override Params()
   )
   > Set Gravity Override(Gravity Override=@n2)
   n3 = GetOwnerSkeletalMeshComponent()
   n4 = BreakRotator(InRot=BreakTransform(InTransform=GetDeltaTransformFromRefPose(self=@n3, BoneName=hand_l, BaseName=lowerarm_l)).Rotation)
   n5 = Branch(execute=@n1.then_6, Condition="float < float"(A=@n4.Yaw))
   > Set Alpha_isHandRotate(Alpha_isHandRotate:float=1) -> (Output_Get:float)
   Set Alpha_isHandRotate(execute=@n5.else, Alpha_isHandRotate=0)
   n6 = GetRealTimeSeconds()
   n7 = Get speedWaveFur() -> (speedWaveFur:double)
   n8 = MultiplyByPi!(Value:double="Sin (Radians)"(A=float * float(A=@n6, B=@n7) {is=title K2Node_PromotableOperator})) -> (ReturnValue:double)
   n9 = Get amplitudWaveFur() -> (amplitudWaveFur:double)
   n10 = "Clamp (Float)"(Value=float + float(A=float * float(A=@n8, B=@n9), B=Get offsetWaveFur() -> (offsetWaveFur:double)), Min=0, Max=100)
   Set rotatorWaveRope_Z(execute=@n1.then_1, rotatorWaveRope_Z:Rotator=MakeRotator(Yaw=@n10)) -> (Output_Get:Rotator)
   > Set rotatorWaveRope_Y(rotatorWaveRope_Y:Rotator=MakeRotator(Pitch=@n10)) -> (Output_Get:Rotator)
   n11 = Branch(execute=@n1.then_11, Condition=OR Boolean!(A=Get bIsDBNO(), B=Get bIsPlayingEmote(), C:bool=Get bIsSurfaceSwimming()) -> (ReturnValue))
   > Set Alpha_IS_DBNO / Emote(Alpha_IS_DBNO / Emote:double=1) -> (Output_Get:double)
   Set Alpha_IS_DBNO / Emote(execute=@n11.else, Alpha_IS_DBNO / Emote=0)
   n12 = Branch(execute=@n1.then_0, Condition=Get bIsSkydiveFloating())
   > Set speedWaveFur(speedWaveFur:double=12) -> (Output_Get:double)
   > Set amplitudWaveFur(amplitudWaveFur:double=2) -> (Output_Get:double)
   > Set offsetWaveFur(offsetWaveFur:double=0) -> (Output_Get:double)
   n13 = Branch(execute=@n12.else, Condition=Get bIsSkydiveDiveMode())
   > Set speedWaveFur(speedWaveFur=30)
   > Set amplitudWaveFur(amplitudWaveFur=1.5)
   > Set offsetWaveFur(offsetWaveFur=0)
   Set speedWaveFur(execute=@n13.else, speedWaveFur=0)
   n14 = Branch(execute=@n1.then_4, Condition=OR Boolean!(A=Get bIsMoving2D(), B=Get bIsFalling()) -> (ReturnValue))
   > Set Alpha_isNecklace_Fall(Alpha_isNecklace_Fall:double=1) -> (Output_Get:double)
   Set Alpha_isNecklace_Fall(execute=@n14.else, Alpha_isNecklace_Fall=0)
   n15 = Branch(execute=@n1.then_5, Condition=Get bIsPlayingMeleeAnim())
   > Set Alpha_isMelee(Alpha_isMelee:double=1) -> (Output_Get:double)
   Set Alpha_isMelee(execute=@n15.else, Alpha_isMelee=0)
   n16 = Branch(execute=@n1.then_3, Condition=Get bIsSkydiveFloating())
   > Set Alpha_isFloating(Alpha_isFloating:float=1) -> (Output_Get:float)
   Set Alpha_isFloating(execute=@n16.else, Alpha_isFloating=0)
   n17 = Branch(execute=@n1.then_7, Condition=Get bIsParachuteOpened())
   > Set Alpha_isOpenParachute(Alpha_isOpenParachute:float=1) -> (Output_Get:float)
   Set Alpha_isOpenParachute(execute=@n17.else, Alpha_isOpenParachute=0)
   n18 = Branch(execute=@n1.then_8, Condition=Get bIsCrouching())
   > Set Alpha_isCrouching(Alpha_isCrouching:float=1) -> (Output_Get:float)
   Set Alpha_isCrouching(execute=@n18.else, Alpha_isCrouching=0)
   n19 = Branch(execute=@n1.then_9, Condition=Get bIsInFrontEnd())
   > Set Alpha_isFrontEnd(Alpha_isFrontEnd:float=0.5) -> (Output_Get:float)
   Set Alpha_isFrontEnd(execute=@n19.else, Alpha_isFrontEnd=1)
   n20 = Cast To FortPlayerAnimInstance(execute=@n1.then_10, Object=GetAnimInstance(self=GetOwnerSkeletalMeshComponent()))
   > Set ar_firingPosition(ar_firingPosition:int=1) -> (Output_Get:int=0)
   > Set AlphaIsSMG_bool(AlphaIsSMG_bool:bool=false) -> (Output_Get:bool)
   > Set Alpha_IsSMG_and_Pistols_Float(Alpha_IsSMG_and_Pistols_Float:double=0) -> (Output_Get:double)
   > n21 = Switch on EFortWeaponCoreAnimation!(execute, Selection:EFortWeaponCoreAnimation Enum=Get WeaponCoreAnim(self=@n20)) -> (
    Melee:Exec
    Pistol:Exec
    Shotgun:Exec
    PaperBlueprint:Exec
    Rifle:Exec
    MeleeOneHand:Exec
    MachinePistol:Exec
    RocketLauncher:Exec
    GrenadeLauncher:Exec
    GoingCommando:Exec
    AssaultRifle:Exec
    TacticalShotgun:Exec
    SniperRifle:Exec
    TrapPlacement:Exec
    ShoulderLauncher:Exec
    AbilityDecoTool:Exec
    Crossbow:Exec
    C4:Exec
    RemoteControl:Exec
    DualWield:Exec
    AR_BullPup:Exec
    AR_ForwardGrip:Exec
    MedPackPaddles:Exec
    SMG_P90:Exec
    AR_DrumGun:Exec
    Consumable_Small:Exec
    Consumable_Large:Exec
    Balloon:Exec
    MountedTurret:Exec
    CreativeTool:Exec
    ExplosiveBow:Exec
    AshtonIndigo:Exec
    AshtonChicago:Exec
    MeleeDualWield:Exec
    Unarmed:Exec
   )
   Set AlphaIsSMG_bool(execute=@n21.Pistol, AlphaIsSMG_bool=true)
   > Set Alpha_IsSMG_and_Pistols_Float(Alpha_IsSMG_and_Pistols_Float=1)
   "float > float"(A=@n4.Yaw, B=9)
  ]}
]
variables [
 Alpha_Emote {type=number}
 Alpha_IS_DBNO / Emote {type=number}
 Alpha_isCrouching {type=number}
 Alpha_isFloating {type=number}
 Alpha_isFrontEnd {type=number}
 Alpha_isHandRotate {type=number}
 Alpha_isMelee {type=number}
 Alpha_isNecklace_Fall {type=number}
 Alpha_isOpenParachute {type=number}
 Alpha_IsSMG_and_Pistols_Float {type=number}
 AlphaIsSMG_bool {type=boolean}
 amplitudWaveFur {type=number}
 ar_firingPosition {type=integer}
 Gravity Override {type=Vector}
 Gravity Override Params {type=GravityOverrideParamsStruct}
 offsetWaveFur {type=number}
 rotatorWaveRope_Y {type=Rotator}
 rotatorWaveRope_Z {type=Rotator}
 speedWaveFur {type=number}
]
