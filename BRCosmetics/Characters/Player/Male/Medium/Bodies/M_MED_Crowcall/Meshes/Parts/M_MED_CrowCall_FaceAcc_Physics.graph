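# Physics asset for the face accessory part of the Crowcall medium male character cosmetic, providing bodies for a dangling hair piece.
# Capsules sit on spine_05, head, hair_base and hair_dyn, with tiny placeholder-sized ones on head and hair_base and larger ones on the spine and hair tip.
# Constraints chain the four bones: spine-to-head and head-to-hair_base swing and twist freely, while hair_dyn has tight swing limits and zero twist relative to hair_base.
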PhysicsAsset /BRCosmetics/Characters/Player/Male/Medium/Bodies/M_MED_Crowcall/Meshes/Parts/M_MED_CrowCall_FaceAcc_Physics
bodies {
 hair_base [{shapeType=Capsule, length=0.505, radius=0.505, shapeName=hair_base_capsule, rotation=<rot 0 -90 0>}]
 hair_dyn [{shapeType=Capsule, length=8.83, radius=2.2951, shapeName=hair_dyn_capsule, center=<1.8113 1.18 0>}]
 head [{shapeType=Capsule, length=0.5, radius=0.5, shapeName=head_capsule}]
 spine_05 [{shapeType=Capsule, length=18.5, radius=5, shapeName=spine_05_capsule, center=<10.5483 -1.0011 0>, rotation=<rot 0 -90 0>}]
}
constraints [
 {
  bone1Name=head
  bone2Name=spine_05
  swing1LimitDegrees=45
  swing1Motion=Free
  swing2LimitDegrees=45
  swing2Motion=Free
  twistLimitDegrees=45
  twistMotion=Free
 }
 {
  bone1Name=hair_base
  bone2Name=head
  swing1LimitDegrees=45
  swing1Motion=Free
  swing2LimitDegrees=45
  swing2Motion=Free
  twistLimitDegrees=45
  twistMotion=Free
 }
 {bone1Name=hair_dyn, bone2Name=hair_base, swing1LimitDegrees=12, swing2LimitDegrees=5, twistLimitDegrees=0}
]
