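# Post-process style AnimBlueprint that animates the hair on the male Crowcall face accessory.
# The attached parent's pose feeds a Control Rig, a RigidBody simulation, then chained spring AnimDynamics on left and right hair bones, sharing a computed gravity override.
# EventGraph sets the rig's assault-rifle input while crouching with an AssaultRifle.
# Its head-yaw alpha drives only an unconnected hair_base transform; the hair_dyn node is also unconnected.
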AnimBlueprint /BRCosmetics/Characters/Player/Male/Medium/Bodies/M_MED_Crowcall/Meshes/Parts/M_MED_CrowCall_FaceAcc_AnimBP
target_skeleton=&/M_MED_CrowCall_FaceAcc_Skeleton
graphs [
 AnimGraph {nodes [
   n1 = Get Gravity Override()
   Anim Dynamics - Bone: hair_dyn!(ComponentPose, Alpha=1, GravityOverride:Vector=@n1, ExternalForce="0,0,0") -> (Pose) {
    angularDampingOverride=1
    angularSpringConstant=20000
    bAngularSpring=true
    bOverrideAngularDamping=true
    bUseGravityOverride=true
    gravityScale=0.35
    simulationSpace=RootRelative
    alphaBoolBlend {blendOption=Linear}
    boundBone {boneName=hair_dyn}
    componentLinearAccScale=<0.5 0.5 0.5>
    componentLinearVelScale=<1 1 1>
    gravityOverride {z=-980}
    physicsBodyDefinitions [{
      boundBone {boneName=hair_dyn}
      boxExtents=<20 30 30>
      constraintSetup {angularLimitsMax=<0 0 10>, angularLimitsMin=<0 0 0>, angularTarget=<1 0 0>}
      localJointOffset=<2.5 0 0>
     }]
   }
   "Transform (Modify) Bone - Bone: hair_base"(Alpha=Get Alpha_isHeadLowRotate() -> (Alpha_isHeadLowRotate:float), Translation="0,0,0", Rotation="0,-30,0", Scale="1,1,1") {
    rotationMode=BMM_Additive
    rotationSpace=BCS_BoneSpace
    scaleMode=BMM_Additive
    scaleSpace=BCS_BoneSpace
    translationMode=BMM_Additive
    translationSpace=BCS_BoneSpace
    alphaBoolBlend {blendOption=Linear}
    alphaScaleBiasClamp {bInterpResult=true, interpSpeedDecreasing=5, interpSpeedIncreasing=5}
    boneToModify {boneName=hair_base}
    rotation {yaw=-30}
   }
   n2 = Control Rig!(
    Source=Copy Pose From Mesh() {bUseAttachedParent=true}
    Alpha
    AssaultRifleSelected:double=Get Alpha_isAssaultRifleSelected_Float() -> (Alpha_isAssaultRifleSelected_Float:float)
   ) -> (Pose) {
    alphaBoolBlend {blendOption=Linear}
    controlRigAssetReference {blueprintRigClass=&/M_MED_CrowCall_FaceAcc_CtrlRig.M_MED_CrowCall_FaceAcc_CtrlRig_C}
   }
   n3 = RigidBody!(ComponentPose=Local To Component(LocalPose=@n2), Alpha, OverrideWorldGravity:Vector=@n1, ExternalForce) -> (Pose) {
    bOverrideWorldGravity=true
    overridePhysicsAsset=&/M_MED_CrowCall_FaceAcc_Physics
    simulationSpace=BaseBoneSpace
    alphaBoolBlend {blendOption=Linear}
    alphaScaleBiasClamp {bInterpResult=true}
    baseBoneRef {boneName=spine_05}
    componentAppliedLinearAccClamp=<3000 3000 3000>
    componentLinearAccScale=<0.3 0.3 0.3>
    componentLinearVelScale=<0.3 0.3 0.3>
    simSpaceSettings {maxAngularAcceleration=1000, maxAngularVelocity=1000, maxLinearAcceleration=1000, maxLinearVelocity=1000}
   }
   n4 = Anim Dynamics - Bone: hair_A_l!(ComponentPose=@n3, Alpha=1, GravityOverride:Vector=@n1, ExternalForce="0,0,0") -> (Pose) {
    angularDampingOverride=0.9999
    angularSpringConstant=20000
    bAngularSpring=true
    bOverrideAngularDamping=true
    bUseGravityOverride=true
    gravityScale=0.35
    simulationSpace=BoneRelative
    alphaBoolBlend {blendOption=Linear}
    boundBone {boneName=hair_A_l}
    componentLinearAccScale=<0.5 0.5 0.5>
    componentLinearVelScale=<0.5 0.5 0.5>
    gravityOverride {z=-980}
    physicsBodyDefinitions [{
      boundBone {boneName=hair_A_l}
      boxExtents=<20 30 30>
      constraintSetup {angularLimitsMax=<15 15 0>, angularLimitsMin=<-15 -15 -15>, angularTarget=<1 0 0>}
      localJointOffset=<2.5 0 0>
     }]
    relativeSpaceBone {boneName=pelvis}
   }
   n5 = Anim Dynamics - Bone: hair_B_l!(ComponentPose=@n4, Alpha=1, GravityOverride:Vector=@n1, ExternalForce="0,0,0") -> (Pose) {
    angularDampingOverride=0.9999
    angularSpringConstant=20000
    bAngularSpring=true
    bOverrideAngularDamping=true
    bUseGravityOverride=true
    gravityScale=0.35
    simulationSpace=BoneRelative
    alphaBoolBlend {blendOption=Linear}
    boundBone {boneName=hair_B_l}
    componentLinearAccScale=<0.5 0.5 0.5>
    componentLinearVelScale=<0.5 0.5 0.5>
    gravityOverride {z=-980}
    physicsBodyDefinitions [{
      boundBone {boneName=hair_B_l}
      boxExtents=<20 30 30>
      constraintSetup {angularLimitsMax=<15 15 0>, angularLimitsMin=<-15 -15 -15>, angularTarget=<1 0 0>}
      localJointOffset=<2.5 0 0>
     }]
    relativeSpaceBone {boneName=pelvis}
   }
   n6 = Anim Dynamics - Bone: hair_C_l!(ComponentPose=@n5, Alpha=1, GravityOverride:Vector=@n1, ExternalForce="0,0,0") -> (Pose) {
    angularDampingOverride=0.9999
    angularSpringConstant=20000
    bAngularSpring=true
    bOverrideAngularDamping=true
    bUseGravityOverride=true
    gravityScale=0.35
    simulationSpace=BoneRelative
    alphaBoolBlend {blendOption=Linear}
    boundBone {boneName=hair_C_l}
    componentLinearAccScale=<0.5 0.5 0.5>
    componentLinearVelScale=<0.5 0.5 0.5>
    gravityOverride {z=-980}
    physicsBodyDefinitions [{
      boundBone {boneName=hair_C_l}
      boxExtents=<20 30 30>
      constraintSetup {angularLimitsMax=<15 15 0>, angularLimitsMin=<-15 -15 -15>, angularTarget=<1 0 0>}
      localJointOffset=<2.5 0 0>
     }]
    relativeSpaceBone {boneName=pelvis}
   }
   n7 = Anim Dynamics - Bone: hair_D_l!(ComponentPose=@n6, Alpha=1, GravityOverride:Vector=@n1, ExternalForce="0,0,0") -> (Pose) {
    angularDampingOverride=0.9999
    angularSpringConstant=20000
    bAngularSpring=true
    bOverrideAngularDamping=true
    bUseGravityOverride=true
    gravityScale=0.35
    simulationSpace=BoneRelative
    alphaBoolBlend {blendOption=Linear}
    boundBone {boneName=hair_D_l}
    componentLinearAccScale=<0.5 0.5 0.5>
    componentLinearVelScale=<0.5 0.5 0.5>
    gravityOverride {z=-980}
    physicsBodyDefinitions [{
      boundBone {boneName=hair_D_l}
      boxExtents=<20 30 30>
      constraintSetup {angularLimitsMax=<15 15 0>, angularLimitsMin=<-15 -15 -15>, angularTarget=<1 0 0>}
      localJointOffset=<2.5 0 0>
     }]
    relativeSpaceBone {boneName=pelvis}
   }
   n8 = Anim Dynamics - Bone: hair_E_l!(ComponentPose=@n7, Alpha=1, GravityOverride:Vector=@n1, ExternalForce="0,0,0") -> (Pose) {
    angularDampingOverride=0.9999
    angularSpringConstant=20000
    bAngularSpring=true
    bOverrideAngularDamping=true
    bUseGravityOverride=true
    gravityScale=0.35
    simulationSpace=BoneRelative
    alphaBoolBlend {blendOption=Linear}
    boundBone {boneName=hair_E_l}
    componentLinearAccScale=<0.5 0.5 0.5>
    componentLinearVelScale=<0.5 0.5 0.5>
    gravityOverride {z=-980}
    physicsBodyDefinitions [{
      boundBone {boneName=hair_E_l}
      boxExtents=<20 30 30>
      constraintSetup {angularLimitsMax=<15 15 0>, angularLimitsMin=<-15 -15 -15>, angularTarget=<1 0 0>}
      localJointOffset=<2.5 0 0>
     }]
    relativeSpaceBone {boneName=pelvis}
   }
   n9 = Anim Dynamics - Bone: hair_A_r!(ComponentPose=@n8, Alpha=1, GravityOverride:Vector=@n1, ExternalForce="0,0,0") -> (Pose) {
    angularDampingOverride=0.9999
    angularSpringConstant=20000
    bAngularSpring=true
    bOverrideAngularDamping=true
    bUseGravityOverride=true
    gravityScale=0.35
    simulationSpace=BoneRelative
    alphaBoolBlend {blendOption=Linear}
    boundBone {boneName=hair_A_r}
    componentLinearAccScale=<0.5 0.5 0.5>
    componentLinearVelScale=<0.5 0.5 0.5>
    gravityOverride {z=-980}
    physicsBodyDefinitions [{
      boundBone {boneName=hair_A_r}
      boxExtents=<20 30 30>
      constraintSetup {angularLimitsMax=<15 15 0>, angularLimitsMin=<-15 -15 -15>, angularTarget=<1 0 0>}
      localJointOffset=<2.5 0 0>
     }]
    relativeSpaceBone {boneName=pelvis}
   }
   n10 = Anim Dynamics - Bone: hair_B_r!(ComponentPose=@n9, Alpha=1, GravityOverride:Vector=@n1, ExternalForce="0,0,0") -> (Pose) {
    angularDampingOverride=0.9999
    angularSpringConstant=20000
    bAngularSpring=true
    bOverrideAngularDamping=true
    bUseGravityOverride=true
    gravityScale=0.35
    simulationSpace=BoneRelative
    alphaBoolBlend {blendOption=Linear}
    boundBone {boneName=hair_B_r}
    componentLinearAccScale=<0.5 0.5 0.5>
    componentLinearVelScale=<0.5 0.5 0.5>
    gravityOverride {z=-980}
    physicsBodyDefinitions [{
      boundBone {boneName=hair_B_r}
      boxExtents=<20 30 30>
      constraintSetup {angularLimitsMax=<15 15 0>, angularLimitsMin=<-15 -15 -15>, angularTarget=<1 0 0>}
      localJointOffset=<2.5 0 0>
     }]
    relativeSpaceBone {boneName=pelvis}
   }
   n11 = Anim Dynamics - Bone: hair_C_r!(ComponentPose=@n10, Alpha=1, GravityOverride:Vector=@n1, ExternalForce="0,0,0") -> (Pose) {
    angularDampingOverride=0.9999
    angularSpringConstant=20000
    bAngularSpring=true
    bOverrideAngularDamping=true
    bUseGravityOverride=true
    gravityScale=0.35
    simulationSpace=BoneRelative
    alphaBoolBlend {blendOption=Linear}
    boundBone {boneName=hair_C_r}
    componentLinearAccScale=<0.5 0.5 0.5>
    componentLinearVelScale=<0.5 0.5 0.5>
    gravityOverride {z=-980}
    physicsBodyDefinitions [{
      boundBone {boneName=hair_C_r}
      boxExtents=<20 30 30>
      constraintSetup {angularLimitsMax=<15 15 0>, angularLimitsMin=<-15 -15 -15>, angularTarget=<1 0 0>}
      localJointOffset=<2.5 0 0>
     }]
    relativeSpaceBone {boneName=pelvis}
   }
   n12 = Anim Dynamics - Bone: hair_D_r!(ComponentPose=@n11, Alpha=1, GravityOverride:Vector=@n1, ExternalForce="0,0,0") -> (Pose) {
    angularDampingOverride=0.9999
    angularSpringConstant=20000
    bAngularSpring=true
    bOverrideAngularDamping=true
    bUseGravityOverride=true
    gravityScale=0.35
    simulationSpace=BoneRelative
    alphaBoolBlend {blendOption=Linear}
    boundBone {boneName=hair_D_r}
    componentLinearAccScale=<0.5 0.5 0.5>
    componentLinearVelScale=<0.5 0.5 0.5>
    gravityOverride {z=-980}
    physicsBodyDefinitions [{
      boundBone {boneName=hair_D_r}
      boxExtents=<20 30 30>
      constraintSetup {angularLimitsMax=<15 15 0>, angularLimitsMin=<-15 -15 -15>, angularTarget=<1 0 0>}
      localJointOffset=<2.5 0 0>
     }]
    relativeSpaceBone {boneName=pelvis}
   }
   n13 = Anim Dynamics - Bone: hair_E_r!(ComponentPose=@n12, Alpha=1, GravityOverride:Vector=@n1, ExternalForce="0,0,0") -> (Pose) {
    angularDampingOverride=0.9999
    angularSpringConstant=20000
    bAngularSpring=true
    bOverrideAngularDamping=true
    bUseGravityOverride=true
    gravityScale=0.35
    simulationSpace=BoneRelative
    alphaBoolBlend {blendOption=Linear}
    boundBone {boneName=hair_E_r}
    componentLinearAccScale=<0.5 0.5 0.5>
    componentLinearVelScale=<0.5 0.5 0.5>
    gravityOverride {z=-980}
    physicsBodyDefinitions [{
      boundBone {boneName=hair_E_r}
      boxExtents=<20 30 30>
      constraintSetup {angularLimitsMax=<15 15 0>, angularLimitsMin=<-15 -15 -15>, angularTarget=<1 0 0>}
      localJointOffset=<2.5 0 0>
     }]
    relativeSpaceBone {boneName=pelvis}
   }
   Output Pose(Result=Component To Local(ComponentPose=@n13)) {layerGroup=None}
  ]}
 EventGraph {nodes [
   Event BlueprintUpdateAnimation()
   > Sequence()
   > n1 = Sequence!(execute) -> (then_0, then_1, then_2:Exec, then_3:Exec, then_4:Exec, then_5:Exec)
   > n2 = Gravity Override State Function DEPRECATED(CharacterPartAnimInstance=Self-Reference(), GameTimeInSeconds=GetGameTimeInSeconds(), GravityOverrideParams=Get Gravity Override Params())
   > Set Gravity Override(Gravity Override=@n2)
   n3 = BreakTransform(InTransform=GetDeltaTransformFromRefPose(self=GetOwnerSkeletalMeshComponent(), BoneName=head, BaseName=neck_01))
   n4 = Branch(execute=@n1.then_4, Condition="float > float"(A=BreakRotator(InRot=@n3.Rotation).Yaw, B=9))
   > Set Alpha_isHeadLowRotate(Alpha_isHeadLowRotate:float=1) -> (Output_Get:float)
   Set Alpha_isHeadLowRotate(execute=@n4.else, Alpha_isHeadLowRotate=0)
   n5 = Cast To FortPlayerAnimInstance(execute=@n1.then_1, Object=GetAnimInstance(self=GetOwnerSkeletalMeshComponent()))
   > Set ar_firingPosition(ar_firingPosition:int=1) -> (Output_Get:int=0)
   > n6 = Switch on EFortWeaponCoreAnimation!(execute, Selection:EFortWeaponCoreAnimation Enum=Get WeaponCoreAnim(self=@n5)) -> (
    Melee:Exec
    Pistol:Exec
    Shotgun:Exec
    PaperBlueprint:Exec
    Rifle:Exec
    MeleeOneHand:Exec
    MachinePistol:Exec
    RocketLauncher:Exec
    GrenadeLauncher:Exec
    GoingCommando:Exec
    AssaultRifle:Exec
    TacticalShotgun:Exec
    SniperRifle:Exec
    TrapPlacement:Exec
    ShoulderLauncher:Exec
    AbilityDecoTool:Exec
    Crossbow:Exec
    C4:Exec
    RemoteControl:Exec
    DualWield:Exec
    AR_BullPup:Exec
    AR_ForwardGrip:Exec
    MedPackPaddles:Exec
    SMG_P90:Exec
    AR_DrumGun:Exec
    Consumable_Small:Exec
    Consumable_Large:Exec
    Balloon:Exec
    MountedTurret:Exec
    CreativeTool:Exec
    ExplosiveBow:Exec
    AshtonIndigo:Exec
    AshtonChicago:Exec
    MeleeDualWield:Exec
    Unarmed:Exec
   )
   Set Alpha_IsAssaultRifleSelected(
    execute=[
     @n6.AR_BullPup
     @n6.AR_DrumGun
     @n6.AR_ForwardGrip
     @n6.AbilityDecoTool
     @n6.AshtonChicago
     @n6.AshtonIndigo
     @n6.Balloon
     @n6.C4
     @n6.Consumable_Large
     @n6.Consumable_Small
     @n6.CreativeTool
     @n6.Crossbow
     @n6.DualWield
     @n6.ExplosiveBow
     @n6.GoingCommando
     @n6.GrenadeLauncher
     @n6.MachinePistol
     @n6.MedPackPaddles
     @n6.Melee
     @n6.MeleeDualWield
     @n6.MeleeOneHand
     @n6.MountedTurret
     @n6.PaperBlueprint
     @n6.Pistol
     @n6.RemoteControl
     @n6.Rifle
     @n6.RocketLauncher
     @n6.SMG_P90
     @n6.Shotgun
     @n6.ShoulderLauncher
     @n6.SniperRifle
     @n6.TacticalShotgun
     @n6.TrapPlacement
     @n6.Unarmed
    ]
    Alpha_IsAssaultRifleSelected:bool=false
   ) -> (Output_Get:bool)
   Set Alpha_IsAssaultRifleSelected(execute=@n6.AssaultRifle, Alpha_IsAssaultRifleSelected=true)
   n7 = Branch(execute=@n1.then_3, Condition=AND Boolean(A=Get bIsCrouching(), B=Get Alpha_IsAssaultRifleSelected() -> (Alpha_IsAssaultRifleSelected:bool)))
   > Set Alpha_isAssaultRifleSelected_Float(Alpha_isAssaultRifleSelected_Float:float=1) -> (Output_Get:float)
   Set Alpha_isAssaultRifleSelected_Float(execute=@n7.else, Alpha_isAssaultRifleSelected_Float=0)
  ]}
]
variables [
 Alpha_IsAssaultRifleRotationY {type=boolean}
 Alpha_IsAssaultRifleSelected {type=boolean}
 Alpha_isAssaultRifleSelected_Float {type=number}
 Alpha_isFloating {type=number}
 Alpha_isHeadLowRotate {type=number}
 ar_firingPosition {type=integer}
 Gravity Override {type=Vector}
 Gravity Override Params {type=GravityOverrideParamsStruct}
]
